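# Control Rig for the FuryFax female medium body that adds secondary deformation for the belt and attached gear.
# After a shared DeformRig function, weighted parent constraints make belt control nulls follow spine, pelvis and leg bones.
# Spherical pose readers on the left thigh blend corrective offsets, then spline IK fits the belt joints to the controls.
# Parent constraints carry grenades, sprays and bags; the head-driven necklace correctives are left unwired.
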ControlRigBlueprint /BRCosmetics/Characters/Player/Female/Medium/Bodies/F_MED_FuryFax/Meshes/F_MED_FuryFax_CtrlRig
graphs [
 New Function_ContainedGraph {nodes [
   n1 = RigVMFunctionEntryNode()
   RigVMFunction_ControlFlowBranch = &RigVMFunction_ControlFlowBranch(ExecuteContext=@n1.ExecuteContext, Condition=@n1.debug)
   RigVMFunctionReturnNode(ExecuteContext=@RigVMFunction_ControlFlowBranch.Completed)
   &RigUnit_SetTransform(
    Item=(Type=Bone,Name="thigh_l")
    Weight=1
    bInitial=False
    Value=&RigUnit_GetTransform(Item=(Type=Control,Name="thigh_l_ctrl"), bInitial=False).Transform
    ExecutePin=@RigVMFunction_ControlFlowBranch.True
   )
  ]}
 New Function_ContainedGraph {nodes [
   RigVMFunctionEntryNode()
   > SplineIK = RigVMFunctionReferenceNode(
    Primary Axis=(X=1,Y=0,Z=0)
    Samples Per Segment=6
    Secondary Spline Direction=(X=0,Y=1,Z=0)
    Spline Mode=BSpline
    Stretch=true
    Up Axis=(X=0,Y=1,Z=0)
    debug=false
    Controls=RigVMVariableNode() {variable=controls}.Value
    Bones=RigVMVariableNode() {variable=joints belt}.Value
   )
   > RigVMFunctionReturnNode()
  ]}
 New Function_ContainedGraph {nodes [
   RigVMFunctionEntryNode()
   > &RigUnit_ParentConstraint(
    AdvancedSettings=(RotationOrderForFilter=XZY)
    Child=(Type=Null,Name="dyn_belt_0_ctrl_null")
    Parents="((Item=(Type=Bone,Name=\"spine_04\"),Weight=0.0033),(Item=(Type=Bone,Name=\"thigh_twist_01_l\"),Weight=0.0076),(Item=(Type=Bone,Name=\"thigh_twist_01_r\"),Weight=0.0076),(Item=(Type=Bone,Name=\"deform_groin_l\"),Weight=0.0081),(Item=(Type=Bone,Name=\"deform_groin_r ...(truncated)"
    Weight=1
   )
   > &RigUnit_ParentConstraint(
    AdvancedSettings=(RotationOrderForFilter=XZY)
    Child=(Type=Null,Name="dyn_belt_1_ctrl_null")
    Parents="((Item=(Type=Bone,Name=\"spine_02\"),Weight=0.0012),(Item=(Type=Bone,Name=\"thigh_twist_01_r\"),Weight=0.0031),(Item=(Type=Bone,Name=\"thigh_twist_01_l\"),Weight=0.1787),(Item=(Type=Bone,Name=\"deform_groin_l\"),Weight=0.2517),(Item=(Type=Bone,Name=\"pelvis\"),Weigh ...(truncated)"
    Weight=1
   )
   > &RigUnit_ParentConstraint(
    AdvancedSettings=(RotationOrderForFilter=XZY)
    Child=(Type=Null,Name="dyn_belt_2_ctrl_null")
    Parents="((Item=(Type=Bone,Name=\"deform_glute_l\"),Weight=0.0001),(Item=(Type=Bone,Name=\"thigh_l\"),Weight=0.045),(Item=(Type=Bone,Name=\"deform_groin_l\"),Weight=0.2253),(Item=(Type=Bone,Name=\"pelvis\"),Weight=0.2698),(Item=(Type=Bone,Name=\"thigh_twist_01_l\"),Weight=0. ...(truncated)"
    Weight=1
   )
   > &RigUnit_ParentConstraint(
    AdvancedSettings=(RotationOrderForFilter=XZY)
    Child=(Type=Null,Name="dyn_belt_3_ctrl_null")
    Parents="((Item=(Type=Bone,Name=\"deform_glute_l\"),Weight=0.0173),(Item=(Type=Bone,Name=\"deform_groin_l\"),Weight=0.0437),(Item=(Type=Bone,Name=\"thigh_l\"),Weight=0.115),(Item=(Type=Bone,Name=\"pelvis\"),Weight=0.1872),(Item=(Type=Bone,Name=\"thigh_twist_01_l\"),Weight=0. ...(truncated)"
    Weight=1
   )
   > &RigUnit_ParentConstraint(
    AdvancedSettings=(RotationOrderForFilter=XZY)
    Child=(Type=Null,Name="dyn_belt_4_ctrl_null")
    Parents="((Item=(Type=Bone,Name=\"deform_groin_l\"),Weight=0.01),(Item=(Type=Bone,Name=\"thigh_l\"),Weight=0.0203),(Item=(Type=Bone,Name=\"deform_glute_l\"),Weight=0.1191),(Item=(Type=Bone,Name=\"pelvis\"),Weight=0.3384),(Item=(Type=Bone,Name=\"thigh_twist_01_l\"),Weight=0.5 ...(truncated)"
    Weight=1
   )
   > &RigUnit_ParentConstraint(
    AdvancedSettings=(RotationOrderForFilter=XZY)
    Child=(Type=Null,Name="dyn_belt_5_ctrl_null")
    Parents=((Item=(Type=Bone,Name="thigh_twist_01_l"),Weight=0.1836),(Item=(Type=Bone,Name="deform_glute_l"),Weight=0.2632),(Item=(Type=Bone,Name="pelvis"),Weight=0.5531))
    Weight=1
   )
   > &RigUnit_ParentConstraint(
    AdvancedSettings=(RotationOrderForFilter=XZY)
    Child=(Type=Null,Name="dyn_belt_6_ctrl_null")
    Parents=((Item=(Type=Bone,Name="thigh_twist_01_l"),Weight=0.0318),(Item=(Type=Bone,Name="deform_glute_l"),Weight=0.2932),(Item=(Type=Bone,Name="pelvis"),Weight=0.6749))
    Weight=1
   )
   > &RigUnit_ParentConstraint(
    AdvancedSettings=(RotationOrderForFilter=XZY)
    Child=(Type=Null,Name="dyn_belt_7_ctrl_null")
    Parents="((Item=(Type=Bone,Name=\"deform_glute_l\"),Weight=0.0008),(Item=(Type=Bone,Name=\"deform_glute_r\"),Weight=0.0008),(Item=(Type=Bone,Name=\"spine_03\"),Weight=0.0014),(Item=(Type=Bone,Name=\"spine_01\"),Weight=0.0222),(Item=(Type=Bone,Name=\"spine_02\"),Weight=0.0372 ...(truncated)"
    Weight=1
   )
   > RigVMFunctionReturnNode()
  ]}
 New Function_ContainedGraph {nodes [
   n1 = RigVMFunctionEntryNode()
   RigVMFunction_Sequence = &RigVMFunction_Sequence(ExecuteContext=@n1.ExecuteContext)
   &RigUnit_ModifyTransforms(
    ItemToModify="((Item=(Type=Null,Name=\"dyn_belt_3_ctrl_null\"),Transform=(Rotation=(X=0,Y=0,Z=0,W=1),Translation=(X=0,Y=0,Z=0.9613),Scale3D=(X=1,Y=1,Z=1))),(Item=(Type=Null,Name=\"dyn_belt_4_ctrl_null\"),Transform=(Rotation=(X=0,Y=0,Z=0,W=1),Translation=(X=0,Y=0,Z=0.4567),S ...(truncated)"
    Weight=1
    WeightMinimum=0
    ExecutePin=@RigVMFunction_Sequence.A
   )
   n2 = &RigUnit_SphericalPoseReader(
    ActiveRegionScaleFactors=(PositiveWidth=3.5,NegativeWidth=0.5,PositiveHeight=3,NegativeHeight=3)
    ActiveRegionSize=0.1
    Debug=(bDrawDebug=false,DebugScale=25,DebugSegments=20,DebugThickness=0.25)
    DriverAxis=(X=-1,Y=0,Z=0)
    DriverItem=(Type=Bone,Name="thigh_l")
    FalloffRegionScaleFactors=(PositiveWidth=1,NegativeWidth=1.26,PositiveHeight=1,NegativeHeight=1)
    FalloffSize=0.45
    OptionalParentItem=(Type=Bone)
    RotationOffset=(X=90,Y=0,Z=-25)
    ExecutePin=@RigVMFunction_Sequence.B
   )
   > &RigUnit_ModifyTransforms(
    ItemToModify="((Item=(Type=Null,Name=\"dyn_belt_0_ctrl_null\"),Transform=(Rotation=(X=0,Y=0,Z=-0.2504,W=0.9681),Translation=(X=-5.6924,Y=-0.2714,Z=1),Scale3D=(X=1,Y=1,Z=1))),(Item=(Type=Null,Name=\"dyn_belt_1_ctrl_null\"),Transform=(Rotation=(X=0.4784,Y=-0.2915,Z=-0.1721,W= ...(truncated)"
    WeightMinimum=0
    Weight=@n2.OutputParam
   )
   > &RigUnit_ModifyTransforms(
    ItemToModify="((Item=(Type=Bone,Name=\"dyn_belt_rot_0\"),Transform=(Rotation=(X=0.6486,Y=0.0844,Z=-0.1945,W=0.7311),Translation=(X=0,Y=0,Z=0),Scale3D=(X=1,Y=1,Z=1))),(Item=(Type=Bone,Name=\"dyn_belt_rot_1\"),Transform=(Rotation=(X=-0.9749,Y=0.0531,Z=0.0466,W=-0.2112),Transl ...(truncated)"
    WeightMinimum=0
    Weight=@n2.OutputParam
   )
   > n3 = &RigUnit_SphericalPoseReader(
    ActiveRegionScaleFactors=(PositiveWidth=1,NegativeWidth=0.4,PositiveHeight=1,NegativeHeight=1)
    ActiveRegionSize=0.1
    Debug=(bDrawDebug=false,bDraw2D=False,bDrawLocalAxes=False,DebugScale=25,DebugSegments=20,DebugThickness=0.25)
    DriverAxis=(X=-1,Y=0,Z=0)
    DriverItem=(Type=Bone,Name="thigh_l")
    FalloffRegionScaleFactors=(PositiveWidth=1,NegativeWidth=2.85,PositiveHeight=1,NegativeHeight=2)
    FalloffSize=0.2
    FlipHeightScaling=False
    FlipWidthScaling=False
    OptionalParentItem=(Type=Bone,Name="None")
    RotationOffset=(X=180,Y=0,Z=0)
   )
   > &RigUnit_ModifyTransforms(
    ItemToModify="((Item=(Type=Null,Name=\"dyn_belt_2_ctrl_null\"),Transform=(Rotation=(X=0,Y=0,Z=0,W=1),Translation=(X=0,Y=1.5014,Z=3.1152),Scale3D=(X=1,Y=1,Z=1))),(Item=(Type=Null,Name=\"dyn_belt_3_ctrl_null\"),Transform=(Rotation=(X=0,Y=0,Z=0,W=1),Translation=(X=-0.5865,Y=3. ...(truncated)"
    WeightMinimum=0
    Weight=@n3.OutputParam
   )
   > n4 = &RigUnit_SphericalPoseReader(
    ActiveRegionScaleFactors=(PositiveWidth=1,NegativeWidth=1,PositiveHeight=1,NegativeHeight=1)
    ActiveRegionSize=0.1
    Debug=(bDrawDebug=false,bDraw2D=False,bDrawLocalAxes=False,DebugScale=25,DebugSegments=20,DebugThickness=0.25)
    DriverAxis=(X=-1,Y=0,Z=0)
    DriverItem=(Type=Bone,Name="thigh_l")
    FalloffRegionScaleFactors=(PositiveWidth=1,NegativeWidth=1,PositiveHeight=1,NegativeHeight=1)
    FalloffSize=0.2
    FlipHeightScaling=False
    FlipWidthScaling=False
    OptionalParentItem=(Type=Bone,Name="None")
    RotationOffset=(X=-90,Y=0,Z=-40)
   )
   > &RigUnit_ModifyTransforms(
    ItemToModify="((Item=(Type=Null,Name=\"dyn_belt_4_ctrl_null\"),Transform=(Rotation=(X=0,Y=0,Z=0,W=1),Translation=(X=0.6297,Y=0,Z=1.0508),Scale3D=(X=1,Y=1,Z=1))),(Item=(Type=Null,Name=\"dyn_belt_5_ctrl_null\"),Transform=(Rotation=(X=0,Y=0,Z=0,W=1),Translation=(X=0,Y=0,Z=1.36 ...(truncated)"
    WeightMinimum=0
    Weight=@n4.OutputParam
   )
   > &RigUnit_ParentConstraint(
    AdvancedSettings=(RotationOrderForFilter=XZY)
    Child=(Type=Bone,Name="dyn_grenade_01")
    Parents="((Item=(Type=Bone,Name=\"dyn_belt_rot_3\"),Weight=0.0007),(Item=(Type=Bone,Name=\"deform_glute_r\"),Weight=0.004),(Item=(Type=Bone,Name=\"spine_01\"),Weight=0.0112),(Item=(Type=Bone,Name=\"spine_02\"),Weight=0.0177),(Item=(Type=Bone,Name=\"dyn_belt_rot_4\"),Weight=0 ...(truncated)"
    Weight=1
   )
   > &RigUnit_ParentConstraint(
    AdvancedSettings=(RotationOrderForFilter=XZY)
    Child=(Type=Bone,Name="dyn_grenade_02")
    Parents="((Item=(Type=Bone,Name=\"spine_03\"),Weight=0.0004),(Item=(Type=Bone,Name=\"dyn_belt_rot_4\"),Weight=0.005),(Item=(Type=Bone,Name=\"deform_glute_r\"),Weight=0.0168),(Item=(Type=Bone,Name=\"spine_01\"),Weight=0.0252),(Item=(Type=Bone,Name=\"spine_02\"),Weight=0.0456) ...(truncated)"
    Weight=1
   )
   > &RigUnit_ParentConstraint(
    AdvancedSettings=(RotationOrderForFilter=XZY)
    Child=(Type=Bone,Name="dyn_spray_01")
    Parents=((Item=(Type=Bone,Name="dyn_belt_7"),Weight=1.0))
    Weight=1
   )
   > &RigUnit_ParentConstraint(
    AdvancedSettings=(RotationOrderForFilter=XZY)
    Child=(Type=Bone,Name="dyn_spray_02")
    Parents="((Item=(Type=Bone,Name=\"dyn_belt_rot_6\"),Weight=0.004),(Item=(Type=Bone,Name=\"deform_groin_r\"),Weight=0.0048),(Item=(Type=Bone,Name=\"deform_glute_r\"),Weight=0.0115),(Item=(Type=Bone,Name=\"spine_03\"),Weight=0.015),(Item=(Type=Bone,Name=\"dyn_belt_7\"),Weight= ...(truncated)"
    Weight=1
   )
   > RigVMFunctionReturnNode()
   n5 = &RigUnit_SphericalPoseReader(
    ActiveRegionScaleFactors=(PositiveWidth=0.75,NegativeWidth=0.75,PositiveHeight=1,NegativeHeight=1)
    ActiveRegionSize=0.1
    Debug=(bDrawDebug=false,bDraw2D=false,bDrawLocalAxes=false,DebugScale=25,DebugSegments=20,DebugThickness=0.25)
    DriverAxis=(X=0,Y=1,Z=0)
    DriverItem=(Type=Bone,Name="head")
    ExecutePin=()
    FalloffRegionScaleFactors=(PositiveWidth=1,NegativeWidth=1,PositiveHeight=1,NegativeHeight=2.28)
    FalloffSize=0.2
    OptionalParentItem=(Type=Bone)
    RotationOffset=(X=150,Y=0,Z=0)
   )
   > RigVMVariableNode(Value=@n5.OutputParam) {variable=head_lookside_l_alpha}
   > n6 = &RigUnit_SphericalPoseReader(
    ActiveRegionScaleFactors=(PositiveWidth=0.75,NegativeWidth=0.75,PositiveHeight=1,NegativeHeight=1)
    ActiveRegionSize=0.1
    Debug=(bDrawDebug=false,DebugScale=25,DebugSegments=20,DebugThickness=0.25)
    DriverAxis=(X=0,Y=1,Z=0)
    DriverItem=(Type=Bone,Name="head")
    FalloffRegionScaleFactors=(PositiveWidth=1,NegativeWidth=1,PositiveHeight=1,NegativeHeight=2.28)
    FalloffSize=0.2
    FlipHeightScaling=true
    FlipWidthScaling=false
    OptionalParentItem=(Type=Bone)
    RotationOffset=(X=30,Y=0,Z=0)
   )
   > RigVMVariableNode(Value=@n6.OutputParam) {variable=head_lookside_r_alpha}
   > n7 = &RigUnit_SphericalPoseReader(
    ActiveRegionScaleFactors=(PositiveWidth=1,NegativeWidth=0.5,PositiveHeight=1.5,NegativeHeight=1.5)
    ActiveRegionSize=0.1
    Debug=(bDrawDebug=false,DebugScale=25,DebugSegments=20,DebugThickness=0.25)
    DriverAxis=(X=0,Y=1,Z=0)
    DriverItem=(Type=Bone,Name="head")
    FalloffRegionScaleFactors=(PositiveWidth=1,NegativeWidth=1.8,PositiveHeight=1.5,NegativeHeight=1.5)
    FalloffSize=0.2
    FlipWidthScaling=false
    OptionalParentItem=(Type=Bone)
    RotationOffset=(X=90,Y=0,Z=-60)
   )
   > RigVMVariableNode(Value=@n7.OutputParam) {variable=head_lookup_alpha}
   > n8 = &RigUnit_SphericalPoseReader(
    ActiveRegionScaleFactors=(PositiveWidth=2,NegativeWidth=0.5,PositiveHeight=2,NegativeHeight=2)
    ActiveRegionSize=0.075
    Debug=(bDrawDebug=false,DebugScale=25,DebugSegments=20,DebugThickness=0.25)
    DriverAxis=(X=0,Y=1,Z=0)
    DriverItem=(Type=Bone,Name="head")
    FalloffRegionScaleFactors=(PositiveWidth=1,NegativeWidth=1.1,PositiveHeight=1,NegativeHeight=1)
    FalloffSize=0.2
    FlipHeightScaling=false
    FlipWidthScaling=true
    OptionalParentItem=(Type=Bone)
    RotationOffset=(X=90,Y=0,Z=42.5)
   )
   > RigVMVariableNode(Value=@n8.OutputParam) {variable=head_lookdw_alpha}
   > &RigUnit_ModifyTransforms(
    ItemToModify=((Item=(Type=Bone,Name="None"),Transform=(Rotation=(X=0,Y=0,Z=0,W=1),Translation=(X=0,Y=0,Z=0),Scale3D=(X=1,Y=1,Z=1))),(Item=(Type=Bone,Name="helper_necklance"),Transform=(Rotation=(X=0,Y=0,Z=0,W=1),Translation=(X=0,Y=1,Z=0),Scale3D=(X=1,Y=1,Z=1))))
    WeightMinimum=0
    Weight=RigVMVariableNode() {variable=head_lookside_r_alpha}.Value
   )
   > &RigUnit_ModifyTransforms(
    ItemToModify=((Item=(Type=Bone,Name="None"),Transform=(Rotation=(X=0,Y=0,Z=0,W=1),Translation=(X=0,Y=0,Z=0),Scale3D=(X=1,Y=1,Z=1))),(Item=(Type=Bone,Name="helper_necklance"),Transform=(Rotation=(X=0,Y=0,Z=0,W=1),Translation=(X=0,Y=0.5,Z=0),Scale3D=(X=1,Y=1,Z=1))))
    WeightMinimum=0
    Weight=RigVMVariableNode() {variable=head_lookside_l_alpha}.Value
   )
   > &RigUnit_ModifyTransforms(
    ItemToModify=((Item=(Type=Bone,Name="None"),Transform=(Rotation=(X=0,Y=0,Z=0,W=1),Translation=(X=0,Y=0,Z=0),Scale3D=(X=1,Y=1,Z=1))),(Item=(Type=Bone,Name="helper_necklance"),Transform=(Rotation=(X=0,Y=0,Z=0,W=1),Translation=(X=0,Y=0.7,Z=0),Scale3D=(X=1,Y=1,Z=1))))
    WeightMinimum=0
    Weight=RigVMVariableNode() {variable=head_lookdw_alpha}.Value
   )
   > &RigUnit_ModifyTransforms(
    ItemToModify=((Item=(Type=Bone,Name="helper_necklance"),Transform=(Rotation=(X=0,Y=0,Z=0.1736,W=0.9848),Translation=(X=-3,Y=0,Z=0),Scale3D=(X=1,Y=1.3,Z=1.3))))
    WeightMinimum=0
    Weight=RigVMVariableNode() {variable=head_lookup_alpha}.Value
   )
  ]}
 RigVMFunctionLibrary {nodes [
   controls = RigVMCollapseNode(ExecuteContext=(), debug=False)
   ikSp_thigh = RigVMCollapseNode(ExecuteContext=())
   pre fixes = RigVMCollapseNode(ExecuteContext=())
   surface constraints = RigVMCollapseNode(ExecuteContext=())
  ]}
 RigVMModel {nodes [
   &RigUnit_ModifyTransforms(
    ExecutePin=()
    ItemToModify="((Item=(Type=Null,Name=\"dyn_belt_3_ctrl_null\"),Transform=(Rotation=(X=0,Y=0,Z=0,W=1),Translation=(X=0,Y=0,Z=0.9613),Scale3D=(X=1,Y=1,Z=1))),(Item=(Type=Null,Name=\"dyn_belt_4_ctrl_null\"),Transform=(Rotation=(X=0,Y=0,Z=0,W=1),Translation=(X=0,Y=0,Z=0.4567),S ...(truncated)"
    Weight=1
    WeightMinimum=0
   )
   RigVMFunction_AnimEvalRichCurve = &RigVMFunction_AnimEvalRichCurve(
    Curve=(EditorCurveData=(Keys=((InterpMode=RCIM_Cubic,TangentMode=RCTM_User,ArriveTangent=0.1206,LeaveTangent=0.1206),(InterpMode=RCIM_Cubic,TangentMode=RCTM_User,Time=1,Value=1,ArriveTangent=1.5305,LeaveTangent=1.5305))))
    SourceMinimum=0
    TargetMinimum=0
    Value=RigVMDispatchNode(BoneName=root, Default=0, Name=deformRig_l_thigh_up_alpha) {template="DISPATCH_RigDispatch_GetAnimAttribute(in Name,in BoneName,in Default,out Value,out Found)"}.Value
   )
   > &RigUnit_ModifyTransforms(
    ItemToModify="((Item=(Type=Null,Name=\"dyn_belt_0_ctrl_null\"),Transform=(Rotation=(X=0,Y=0,Z=0,W=1),Translation=(X=1.6023,Y=0.5022,Z=0.4498),Scale3D=(X=1,Y=1,Z=1))),(Item=(Type=Null,Name=\"dyn_belt_1_ctrl_null\"),Transform=(Rotation=(X=0.5085,Y=0,Z=0,W=0.8611),Translation= ...(truncated)"
    WeightMinimum=0
    Weight=@RigVMFunction_AnimEvalRichCurve.Result
   )
   > &RigUnit_ModifyTransforms(
    ItemToModify="((Item=(Type=Bone,Name=\"dyn_belt_rot_0\"),Transform=(Rotation=(X=0.6486,Y=0.0844,Z=-0.1945,W=0.7311),Translation=(X=0,Y=0,Z=0),Scale3D=(X=1,Y=1,Z=1))),(Item=(Type=Bone,Name=\"dyn_belt_rot_1\"),Transform=(Rotation=(X=-0.8617,Y=0,Z=0,W=-0.5074),Translation=(X=0 ...(truncated)"
    WeightMinimum=0
    Weight=@RigVMFunction_AnimEvalRichCurve.Result
   )
   > n1 = &RigUnit_SphericalPoseReader(
    ActiveRegionScaleFactors=(PositiveWidth=1,NegativeWidth=0.4,PositiveHeight=1,NegativeHeight=1)
    ActiveRegionSize=0.1
    Debug=(bDrawDebug=false,bDraw2D=False,bDrawLocalAxes=False,DebugScale=25,DebugSegments=20,DebugThickness=0.25)
    DriverAxis=(X=-1,Y=0,Z=0)
    DriverItem=(Type=Bone,Name="thigh_l")
    FalloffRegionScaleFactors=(PositiveWidth=1,NegativeWidth=3,PositiveHeight=1,NegativeHeight=1)
    FalloffSize=0.2
    FlipHeightScaling=False
    FlipWidthScaling=False
    OptionalParentItem=(Type=Bone,Name="None")
    RotationOffset=(X=180,Y=0,Z=0)
   )
   > &RigUnit_ModifyTransforms(
    ItemToModify="((Item=(Type=Null,Name=\"dyn_belt_2_ctrl_null\"),Transform=(Rotation=(X=0,Y=0,Z=0,W=1),Translation=(X=0,Y=0,Z=3.1152),Scale3D=(X=1,Y=1,Z=1))),(Item=(Type=Null,Name=\"dyn_belt_3_ctrl_null\"),Transform=(Rotation=(X=0,Y=0,Z=0,W=1),Translation=(X=0,Y=0,Z=5.8918),S ...(truncated)"
    WeightMinimum=0
    Weight=@n1.OutputParam
   )
   > n2 = &RigUnit_SphericalPoseReader(
    ActiveRegionScaleFactors=(PositiveWidth=1,NegativeWidth=1,PositiveHeight=1,NegativeHeight=1)
    ActiveRegionSize=0.1
    Debug=(bDrawDebug=false,bDraw2D=False,bDrawLocalAxes=False,DebugScale=25,DebugSegments=20,DebugThickness=0.25)
    DriverAxis=(X=-1,Y=0,Z=0)
    DriverItem=(Type=Bone,Name="thigh_l")
    FalloffRegionScaleFactors=(PositiveWidth=1,NegativeWidth=1,PositiveHeight=1,NegativeHeight=1)
    FalloffSize=0.2
    FlipHeightScaling=False
    FlipWidthScaling=False
    OptionalParentItem=(Type=Bone,Name="None")
    RotationOffset=(X=-90,Y=0,Z=-40)
   )
   > &RigUnit_ModifyTransforms(
    ItemToModify="((Item=(Type=Null,Name=\"dyn_belt_4_ctrl_null\"),Transform=(Rotation=(X=0,Y=0,Z=0,W=1),Translation=(X=0.6297,Y=0,Z=3.5602),Scale3D=(X=1,Y=1,Z=1))),(Item=(Type=Null,Name=\"dyn_belt_5_ctrl_null\"),Transform=(Rotation=(X=0,Y=0,Z=0,W=1),Translation=(X=0,Y=0,Z=1.36 ...(truncated)"
    WeightMinimum=0
    Weight=@n2.OutputParam
   )
   RigUnit_BeginExecution = &RigUnit_BeginExecution()
   RigVMFunction_Sequence = &RigVMFunction_Sequence(ExecuteContext=@RigUnit_BeginExecution.ExecutePin)
   FN_DeformRig_FNC = RigVMFunctionReferenceNode(
    DeformRig_ Curve="(deformRig_l_upperarm_up_alpha_72_#1=False,deformRig_r_upperarm_up_alpha_73_#2=False,deformRig_l_upperarm_fr_alpha_74_#3=False,deformRig_r_upperarm_fr_alpha_75_CB51B8 ...(truncated)"
    DeformRig_Bool="(upperarm_1_#4=True,elbow_3_#5=True,wrist_7_#6=True,groin_8_#7=True,glute_17_#8=True,knee_9_40C408324CA261 ...(truncated)"
    DeformRig_Debug=(upperarm_debug_18_#4=False,wrist_debug_20_#6=False,lat_debug_26_#9=false,groinGlute_debug_27_#7=false)
    DeformRig_Value="(upperarm_up_18_#10=(Rotation=(X=0.0,Y=-0.0,Z=0.0,W=1.0),Translation=(X=0.0,Y=0.0,Z=8.0),Scale3D=(X=0.70,Y=1.0,Z=1.0)),upperarm_fwd_fr_19_#11=(Rotation=(X=0.0,Y=-0.0,Z=0.0,W=1.0),Translation=(X=0.0, ...(truncated)"
    ExecuteContext=@RigVMFunction_Sequence.A
   )
   surface constraints = RigVMFunctionReferenceNode(ExecuteContext=@RigVMFunction_Sequence.B)
   > pre fixes = RigVMFunctionReferenceNode()
   > ikSp_thigh = RigVMFunctionReferenceNode()
   > &RigUnit_ParentConstraint(
    AdvancedSettings=(RotationOrderForFilter=XZY)
    Child=(Type=Bone,Name="base_bag_01_l")
    Parents="((Item=(Type=Bone,Name=\"pelvis\"),Weight=0.0003),(Item=(Type=Bone,Name=\"dyn_belt_7\"),Weight=0.001),(Item=(Type=Bone,Name=\"dyn_belt_rot_0\"),Weight=0.0073),(Item=(Type=Bone,Name=\"dyn_belt_rot_6\"),Weight=0.017),(Item=(Type=Bone,Name=\"dyn_belt_rot_1\"),Weight=0. ...(truncated)"
    Weight=1
   )
   > &RigUnit_ParentConstraint(
    AdvancedSettings=(RotationOrderForFilter=XZY)
    Child=(Type=Bone,Name="base_bag_02_l")
    Parents="((Item=(Type=Bone,Name=\"thigh_twist_01_r\"),Weight=0.0001),(Item=(Type=Bone,Name=\"dyn_belt_rot_6\"),Weight=0.0001),(Item=(Type=Bone,Name=\"spine_04\"),Weight=0.0002),(Item=(Type=Bone,Name=\"deform_groin_r\"),Weight=0.0002),(Item=(Type=Bone,Name=\"spine_02\"),Weigh ...(truncated)"
    Weight=1
   )
  ]}
]
hierarchy [
 root {loc=<0 0 0>, rot=<0 0 0 1>}
 pelvis {parent=root, loc=<0 -0.5301 98.0571>, rot=<0.0127 0.707 -0.0127 -0.707>}
 spine_01 {parent=pelvis, loc=<6.1208 -0.9328 0>, rot=<0 0 0.034 -0.9994>}
 spine_02 {parent=spine_01, loc=<6.5996 -0.2095 0>, rot=<0 0 0.0238 -0.9997>}
 spine_03 {parent=spine_02, loc=<6.4914 -0.1237 0>, rot=<0 0 0.0149 -0.9999>}
 spine_04 {parent=spine_03, loc=<6.9598 0.0794 0>, rot=<0 0 -0.0132 -0.9999>}
 spine_05 {parent=spine_04, loc=<6.9987 0 0>, rot=<0 0 -0.0015 -1>}
 clavicle_l {parent=spine_05, loc=<5.4831 -0.1053 -3.2946>, rot=<0.0465 -0.7289 -0.009 -0.6829>}
 upperarm_l {parent=clavicle_l, loc=<9.8658 -1.2136 -0.4105>, rot=<0.0164 -0.3245 -0.0269 -0.9454>}
 lowerarm_l {parent=upperarm_l, loc=<26.465 -0.8345 0.259>, rot=<0.0036 0.0161 -0.2298 -0.9731>}
 hand_l {parent=lowerarm_l, loc=<22.7355 -0.3203 -0.6105>, rot=<-0.6784 -0.0223 0.0735 -0.7307>}
 index_metacarpal_l {parent=hand_l, loc=<1.6535 0.0183 -2.96>, rot=<0.0158 -0.0124 -0.0173 -0.9996>}
 index_01_l {parent=index_metacarpal_l, loc=<8.3008 0.0357 -0.0524>, rot=<-0.0044 0.0485 0.0924 -0.9945>}
 index_02_l {parent=index_01_l, loc=<3.3587 -0.3004 0.0452>, rot=<-0.002 0.0027 0.3225 -0.9466>}
 index_03_l {parent=index_02_l, loc=<2.5578 -0.123 -0.0252>, rot=<-0.002 0.0067 0.2558 -0.9667>}
 middle_metacarpal_l {parent=hand_l, loc=<2.1856 -0.1725 -0.5984>, rot=<0.006 0.0168 -0.0329 -0.9993>}
 middle_01_l {parent=middle_metacarpal_l, loc=<7.3275 -0.2447 0.0129>, rot=<-0.0208 0.0042 0.0858 -0.9961>}
 middle_02_l {parent=middle_01_l, loc=<4.0123 0.1401 0.1175>, rot=<-0.0218 0.0258 0.3512 -0.9357>}
 middle_03_l {parent=middle_02_l, loc=<2.6077 0.0534 0.1373>, rot=<0.0082 0.0188 0.1708 -0.9851>}
 pinky_metacarpal_l {parent=hand_l, loc=<1.7616 -0.9753 3.3814>, rot=<-0.0397 0.0814 -0.069 -0.9935>}
 pinky_01_l {parent=pinky_metacarpal_l, loc=<7.1709 -0.4831 -0.1094>, rot=<-0.0654 -0.0035 0.2242 -0.9723>}
 pinky_02_l {parent=pinky_01_l, loc=<2.8274 0.0893 -0.1043>, rot=<0.0037 0.0011 0.2854 -0.9584>}
 pinky_03_l {parent=pinky_02_l, loc=<2.2046 0.0307 -0.0078>, rot=<-0.0002 0.0098 0.1445 -0.9895>}
 ring_metacarpal_l {parent=hand_l, loc=<2.2074 -0.3625 1.3638>, rot=<-0.0284 0.0722 -0.0458 -0.9959>}
 ring_01_l {parent=ring_metacarpal_l, loc=<7.0295 -0.6197 -0.1446>, rot=<-0.0489 -0.0252 0.1892 -0.9804>}
 ring_02_l {parent=ring_01_l, loc=<3.6174 0.4647 -0.1027>, rot=<0 0 0.2806 -0.9598>}
 ring_03_l {parent=ring_02_l, loc=<2.4152 0.114 0.0199>, rot=<0 0 0.1189 -0.9929>}
 thumb_01_l {parent=hand_l, loc=<1.3118 -0.3191 -2.459>, rot=<0.6546 -0.4042 0.0664 -0.6354>}
 thumb_02_l {parent=thumb_01_l, loc=<5.9336 -0.0448 0.0401>, rot=<-0.0037 0.0213 0.298 -0.9543>}
 thumb_03_l {parent=thumb_02_l, loc=<2.2223 0.3682 -0.0907>, rot=<0.0118 -0.013 0.0066 -0.9998>}
 deform_wrist_dn_base_l {parent=hand_l, loc=<0 -1.5326 0>, rot=<0 0 0 -1>}
 deform_wrist_dn_l {parent=deform_wrist_dn_base_l, loc=<0 0 0>, rot=<0 0 0 1>}
 deform_wrist_up_base_l {parent=hand_l, loc=<0 2.3009 0>, rot=<0 0 0 -1>}
 deform_wrist_up_l {parent=deform_wrist_up_base_l, loc=<0 0 0>, rot=<0 0 0 1>}
 weapon_l {parent=hand_l, loc=<-0.1974 -3.4111 0.3811>, rot=<0 0 0 1>}
 lowerarm_twist_01_l {parent=lowerarm_l, loc=<18.316 0 0>, rot=<0 0 0 1>}
 lowerarm_twist_02_l {parent=lowerarm_l, loc=<10.6707 0 0>, rot=<0 0 0 1>}
 deform_elbow_in_base_l {parent=lowerarm_l, loc=<0.1834 3.6062 -0.2572>, rot=<0 0 0 -1>}
 deform_elbow_in_l {parent=deform_elbow_in_base_l, loc=<0 0 0>, rot=<0 0 0 1>}
 deform_elbow_out_base_l {parent=lowerarm_l, loc=<0 -4 0>, rot=<0 0 0 1>}
 deform_elbow_out_l {parent=deform_elbow_out_base_l, loc=<0 0 0>, rot=<0 0 0 1>}
 upperarm_twist_01_l {parent=upperarm_l, loc=<7.2135 -0.1134 0>, rot=<0 0 0 1>}
 deform_upperarm_bk_base_l {parent=upperarm_twist_01_l, loc=<-6.4053 -4.3746 0.314>, rot=<0 0 0 -1>}
 deform_upperarm_bk_l {parent=deform_upperarm_bk_base_l, loc=<0 0 0>, rot=<0 0 0 1>}
 deform_upperarm_fr_base_l {parent=upperarm_twist_01_l, loc=<-5.6312 5.092 0.3354>, rot=<0 0 0 -1>}
 deform_upperarm_fr_l {parent=deform_upperarm_fr_base_l, loc=<0 0 0>, rot=<0 0 0 1>}
 deform_upperarm_up_base_l {parent=upperarm_twist_01_l, loc=<-8.891 0.4825 3.9517>, rot=<0 0 0 -1>}
 deform_upperarm_up_l {parent=deform_upperarm_up_base_l, loc=<0 0 0>, rot=<0 0 0 1>}
 upperarm_twist_02_l {parent=upperarm_l, loc=<19.7998 -0.1134 0>, rot=<0 0 0 1>}
 clavicle_r {parent=spine_05, loc=<5.4831 -0.1053 3.2946>, rot=<0.7289 0.0465 -0.6829 0.009>}
 upperarm_r {parent=clavicle_r, loc=<-9.8658 1.2136 0.4104>, rot=<0.0164 -0.3245 -0.0269 -0.9454>}
 lowerarm_r {parent=upperarm_r, loc=<-26.4651 0.8345 -0.259>, rot=<0.0036 0.0161 -0.2298 -0.9731>}
 hand_r {parent=lowerarm_r, loc=<-22.7356 0.3203 0.6106>, rot=<-0.6784 -0.0223 0.0735 -0.7307>}
 index_metacarpal_r {parent=hand_r, loc=<-1.6534 -0.0181 2.9601>, rot=<0.0158 -0.0124 -0.0173 -0.9996>}
 index_01_r {parent=index_metacarpal_r, loc=<-8.3008 -0.0357 0.0524>, rot=<-0.0044 0.0485 0.0924 -0.9945>}
 index_02_r {parent=index_01_r, loc=<-3.3589 0.3003 -0.0453>, rot=<-0.002 0.0027 0.3225 -0.9466>}
 index_03_r {parent=index_02_r, loc=<-2.5578 0.123 0.0252>, rot=<-0.002 0.0067 0.2558 -0.9667>}
 middle_metacarpal_r {parent=hand_r, loc=<-2.1856 0.1726 0.5984>, rot=<0.006 0.0168 -0.0329 -0.9993>}
 middle_01_r {parent=middle_metacarpal_r, loc=<-7.3273 0.2447 -0.0129>, rot=<-0.0208 0.0042 0.0858 -0.9961>}
 middle_02_r {parent=middle_01_r, loc=<-4.0124 -0.14 -0.1175>, rot=<-0.0218 0.0258 0.3512 -0.9357>}
 middle_03_r {parent=middle_02_r, loc=<-2.6077 -0.0533 -0.1373>, rot=<0.0082 0.0188 0.1708 -0.9851>}
 pinky_metacarpal_r {parent=hand_r, loc=<-1.7617 0.9753 -3.3814>, rot=<-0.0397 0.0814 -0.069 -0.9935>}
 pinky_01_r {parent=pinky_metacarpal_r, loc=<-7.1707 0.4832 0.1094>, rot=<-0.0654 -0.0035 0.2242 -0.9723>}
 pinky_02_r {parent=pinky_01_r, loc=<-2.8274 -0.0892 0.1043>, rot=<0.0037 0.0011 0.2854 -0.9584>}
 pinky_03_r {parent=pinky_02_r, loc=<-2.2045 -0.031 0.0078>, rot=<-0.0002 0.0098 0.1445 -0.9895>}
 ring_metacarpal_r {parent=hand_r, loc=<-2.2073 0.3627 -1.3638>, rot=<-0.0284 0.0722 -0.0458 -0.9959>}
 ring_01_r {parent=ring_metacarpal_r, loc=<-7.0297 0.6197 0.1446>, rot=<-0.0489 -0.0252 0.1892 -0.9804>}
 ring_02_r {parent=ring_01_r, loc=<-3.6171 -0.4648 0.1027>, rot=<0 0 0.2806 -0.9598>}
 ring_03_r {parent=ring_02_r, loc=<-2.4153 -0.1141 -0.0199>, rot=<0 0 0.1189 -0.9929>}
 thumb_01_r {parent=hand_r, loc=<-1.3119 0.3192 2.459>, rot=<0.6546 -0.4042 0.0664 -0.6354>}
 thumb_02_r {parent=thumb_01_r, loc=<-5.9335 0.0447 -0.0401>, rot=<-0.0037 0.0213 0.298 -0.9543>}
 thumb_03_r {parent=thumb_02_r, loc=<-2.2223 -0.3682 0.0907>, rot=<0.0118 -0.013 0.0066 -0.9998>}
 deform_wrist_dn_base_r {parent=hand_r, loc=<0 1.5329 0>, rot=<0 0 0 -1>}
 deform_wrist_dn_r {parent=deform_wrist_dn_base_r, loc=<0 0 0>, rot=<0 0 0 1>}
 deform_wrist_up_base_r {parent=hand_r, loc=<-0.0001 -2.3006 -0.0001>, rot=<0 0 0 -1>}
 deform_wrist_up_r {parent=deform_wrist_up_base_r, loc=<0 0 0>, rot=<0 0 0 1>}
 weapon_r {parent=hand_r, loc=<1.9909 -0.2363 -0.9772>, rot=<-0.0219 -0.0515 0.9964 0.0636>}
 lowerarm_twist_01_r {parent=lowerarm_r, loc=<-18.3161 0 0>, rot=<0 0 0 1>}
 lowerarm_twist_02_r {parent=lowerarm_r, loc=<-10.6707 0 0>, rot=<0 0 0 1>}
 deform_elbow_in_base_r {parent=lowerarm_r, loc=<-0.1834 -3.6062 0.2572>, rot=<0 0 0 -1>}
 deform_elbow_in_r {parent=deform_elbow_in_base_r, loc=<0 0 0>, rot=<0 0 0 1>}
 deform_elbow_out_base_r {parent=lowerarm_r, loc=<0 4 0>, rot=<0 0 0 1>}
 deform_elbow_out_r {parent=deform_elbow_out_base_r, loc=<0 0 0>, rot=<0 0 0 1>}
 upperarm_twist_01_r {parent=upperarm_r, loc=<-7.2135 0.1134 0>, rot=<0 0 0 1>}
 deform_upperarm_bk_base_r {parent=upperarm_twist_01_r, loc=<6.4052 4.3746 -0.3139>, rot=<0 0 0 -1>}
 deform_upperarm_bk_r {parent=deform_upperarm_bk_base_r, loc=<0 0 0>, rot=<0 0 0 1>}
 deform_upperarm_fr_base_r {parent=upperarm_twist_01_r, loc=<5.6311 -5.092 -0.3353>, rot=<0 0 0 -1>}
 deform_upperarm_fr_r {parent=deform_upperarm_fr_base_r, loc=<0 0 0>, rot=<0 0 0 1>}
 deform_upperarm_up_base_r {parent=upperarm_twist_01_r, loc=<8.8909 -0.4825 -3.9516>, rot=<0 0 0 -1>}
 deform_upperarm_up_r {parent=deform_upperarm_up_base_r, loc=<0 0 0>, rot=<0 0 0 1>}
 upperarm_twist_02_r {parent=upperarm_r, loc=<-19.7998 0.1134 0>, rot=<0 0 0 1>}
 neck_01 {parent=spine_05, loc=<11.3138 0.1529 0>, rot=<0 0 -0.2003 -0.9797>}
 neck_02 {parent=neck_01, loc=<4.5627 0 0>, rot=<0 0 0.0504 -0.9987>}
 head {parent=neck_02, loc=<5.3914 0 0>, rot=<0 0 0.115 -0.9934>}
 FACIAL_C_FacialRoot {parent=head, loc=<1.9362 0.8413 0>, rot=<0.5021 -0.4979 -0.5021 -0.4979>}
 FACIAL_C_UnderChin {parent=FACIAL_C_FacialRoot, loc=<0 7.7819 2.9063>, rot=<0.5 0 0 -0.866>}
 FACIAL_L_UnderChin {parent=FACIAL_C_FacialRoot, loc=<2.637 6.8418 1.9024>, rot=<0.4025 -0.355 0.0823 -0.8398>}
 FACIAL_R_UnderChin {parent=FACIAL_C_FacialRoot, loc=<-2.6367 6.8417 1.9032>, rot=<0.4025 0.355 -0.0823 -0.8398>}
 FACIAL_C_Jaw {parent=FACIAL_C_FacialRoot, loc=<0 2.1908 1.1574>, rot=<0.342 0 0 -0.9397>}
 FACIAL_C_Jawline {parent=FACIAL_C_Jaw, loc=<0.0006 1.8788 7.2438>, rot=<0.2588 0 0 -0.9659>}
 FACIAL_C_Chin {parent=FACIAL_C_Jaw, loc=<0.0008 0.7991 7.3594>, rot=<-0.2588 0 0 -0.9659>}
 FACIAL_L_ChinSide {parent=FACIAL_C_Jaw, loc=<2.5126 0.1047 7.0704>, rot=<-0.2207 -0.3739 0.05 -0.8994>}
 FACIAL_R_ChinSide {parent=FACIAL_C_Jaw, loc=<-2.5126 0.1047 7.0704>, rot=<-0.2207 0.3739 -0.05 -0.8994>}
 FACIAL_L_Jawline {parent=FACIAL_C_FacialRoot, loc=<3.8788 5.3455 3.3203>, rot=<0.2092 0.1578 -0.0547 -0.9635>}
 FACIAL_R_Jawline {parent=FACIAL_C_FacialRoot, loc=<-3.8789 5.3455 3.3203>, rot=<0.2092 -0.1578 0.0547 -0.9635>}
 helper_necklance {parent=neck_02, loc=<-3.261 4.9051 0>, rot=<0 0 0 1>}
 helper_necklance_l {parent=neck_02, loc=<-3.2157 2.7156 -3.5554>, rot=<0.5106 0.0483 0.0752 -0.8552>}
 base_necklace {parent=spine_05, loc=<6.7216 5.5552 0>, rot=<0 0 0.9517 0.3071>}
 dyn_necklace_short_01 {parent=base_necklace, loc=<0 0 0>, rot=<0 0 0 1>}
 dyn_necklace_short_02 {parent=dyn_necklace_short_01, loc=<7.6835 0 0>, rot=<0 0 -0.044 -0.999>}
 dyn_necklace_long_01 {parent=base_necklace, loc=<0 0 0>, rot=<0 0 0 1>}
 dyn_necklace_long_02 {parent=dyn_necklace_long_01, loc=<0 0 0>, rot=<0 0 0 1>}
 dyn_necklace_long_03 {parent=dyn_necklace_long_02, loc=<7.745 0.0054 0>, rot=<0 0 -0.044 -0.999>}
 dyn_necklace_long_medal_04 {parent=dyn_necklace_long_03, loc=<2.6667 0 0>, rot=<0 0 0 1>}
 dyn_necklace_long_04 {parent=dyn_necklace_long_03, loc=<3.0564 0 0>, rot=<0 0 0 1>}
 dyn_necklace_long_medal_05 {parent=dyn_necklace_long_04, loc=<2.7882 0.436 0>, rot=<0 0 -0.1474 -0.9891>}
 attach_cape {parent=spine_05, loc=<9.3216 -8.4501 0>, rot=<0.0616 -0.7044 -0.0616 -0.7044>}
 attach_backpack {parent=spine_05, loc=<0 -8.35 0>, rot=<-0.0093 -0.707 0.0093 -0.707>}
 deform_pec_aim_base_l {parent=spine_04, loc=<3.2456 9.3309 -1.2237>, rot=<0.0603 -0.4033 0.3318 -0.8507>}
 deform_pec_aim_l {parent=deform_pec_aim_base_l, loc=<0 0 0>, rot=<0 0 0 1>}
 deform_pec_base_l {parent=deform_pec_aim_l, loc=<9.3665 0 0>, rot=<0 0 0 1>}
 deform_pec_l {parent=deform_pec_base_l, loc=<0 0 0>, rot=<0 0 0 1>}
 deform_pec_aim_base_r {parent=spine_04, loc=<3.2452 9.3309 1.2237>, rot=<-0.4033 -0.0603 0.8507 0.3318>}
 deform_pec_aim_r {parent=deform_pec_aim_base_r, loc=<0 0 0>, rot=<0 0 0 1>}
 deform_pec_base_r {parent=deform_pec_aim_r, loc=<-9.3667 -0.0003 -0.0004>, rot=<0 0 0 1>}
 deform_pec_r {parent=deform_pec_base_r, loc=<0 0 0>, rot=<0 0 0 1>}
 thigh_l {parent=pelvis, loc=<-1.0855 -0.2109 -8.7053>, rot=<0.0009 0.0355 0.0155 -0.9993>}
 calf_l {parent=thigh_l, loc=<-42.5284 -0.0019 -0.7297>, rot=<0.0006 -0.0161 -0.0391 -0.9991>}
 calf_twist_01_l {parent=calf_l, loc=<-35.8483 -1.9197 -0.5794>, rot=<0.0237 0.0095 -0.0035 -0.9997>}
 base_zip_shoes_l {parent=calf_twist_01_l, loc=<-5.2176 2.7257 3.8934>, rot=<-0.1096 -0.6059 0.7404 0.2696>}
 dyn_zip_shoes_l {parent=base_zip_shoes_l, loc=<0 0 0>, rot=<0 0 0 1>}
 calf_twist_02_l {parent=calf_l, loc=<-19.1546 -1.7943 -0.2638>, rot=<0.0237 0.0095 -0.0035 -0.9997>}
 foot_l {parent=calf_l, loc=<-46.7402 -2.5014 -0.592>, rot=<0.0427 -0.0238 0.0353 -0.9982>}
 ball_l {parent=foot_l, loc=<-3.5658 10.8066 -0.1544>, rot=<0.0001 -0.0001 -0.7063 -0.7079>}
 deform_knee_in_base_l {parent=calf_l, loc=<0.8031 -5.0231 0.2456>, rot=<0 0 0 -1>}
 deform_knee_in_l {parent=deform_knee_in_base_l, loc=<0 0 0>, rot=<0 0 0 1>}
 deform_knee_out_base_l {parent=calf_l, loc=<0 4 0>, rot=<0 0 0 1>}
 deform_knee_out_l {parent=deform_knee_out_base_l, loc=<0 0 0>, rot=<0 0 0 1>}
 thigh_twist_01_l {parent=thigh_l, loc=<-27.952 0 0>, rot=<0.0013 0 0 -1>}
 deform_groin_base_l {parent=thigh_l, loc=<0.4801 5.6776 -0.2486>, rot=<0 0 0 -1>}
 deform_groin_l {parent=deform_groin_base_l, loc=<0 0 0>, rot=<0 0 0 1>}
 thigh_r {parent=pelvis, loc=<-1.0853 -0.2109 8.7053>, rot=<0.0355 -0.0009 0.9993 0.0155>}
 calf_r {parent=thigh_r, loc=<42.5286 0.0019 0.7297>, rot=<0.0006 -0.0161 -0.0391 -0.9991>}
 calf_twist_01_r {parent=calf_r, loc=<35.8485 1.9197 0.5794>, rot=<0.0237 0.0095 -0.0035 -0.9997>}
 base_zip_shoes_r {parent=calf_twist_01_r, loc=<5.2891 -2.7579 -3.9178>, rot=<-0.6065 0.1113 0.284 -0.7343>}
 dyn_zip_shoes_r {parent=base_zip_shoes_r, loc=<0 0 0>, rot=<0 0 0 1>}
 calf_twist_02_r {parent=calf_r, loc=<19.1546 1.7943 0.2638>, rot=<0.0237 0.0095 -0.0035 -0.9997>}
 foot_r {parent=calf_r, loc=<46.7403 2.5014 0.592>, rot=<0.0427 -0.0238 0.0353 -0.9982>}
 ball_r {parent=foot_r, loc=<3.5658 -10.8066 0.1544>, rot=<0.0001 -0.0001 -0.7063 -0.7079>}
 deform_knee_in_base_r {parent=calf_r, loc=<-0.8031 5.0231 -0.2457>, rot=<0 0 0 -1>}
 deform_knee_in_r {parent=deform_knee_in_base_r, loc=<0 0 0>, rot=<0 0 0 1>}
 deform_knee_out_base_r {parent=calf_r, loc=<0 -4 0>, rot=<0 0 0 1>}
 deform_knee_out_r {parent=deform_knee_out_base_r, loc=<0 0 0>, rot=<0 0 0 1>}
 thigh_twist_01_r {parent=thigh_r, loc=<27.952 0 0>, rot=<0.0013 0 0 -1>}
 deform_groin_base_r {parent=thigh_r, loc=<-0.4799 -5.6776 0.2485>, rot=<0 0 0 -1>}
 deform_groin_r {parent=deform_groin_base_r, loc=<0 0 0>, rot=<0 0 0 1>}
 base_bag_r {parent=thigh_r, loc=<15.1174 -2.3485 7.9227>, rot=<0.683 0.0088 0.0164 -0.7302>}
 dyn_bag_r {parent=base_bag_r, loc=<0 0 0>, rot=<0 0 0 1>}
 base_zipper_bag_r {parent=dyn_bag_r, loc=<-7.0651 -8.7811 6.5097>, rot=<0.7038 0.0841 0.1196 -0.6952>}
 dyn_zipper_bag_r {parent=base_zipper_bag_r, loc=<0 0 0>, rot=<0 0 0 1>}
 base_bag_01_r {parent=dyn_bag_r, loc=<4.4463 -7.4272 -3.1765>, rot=<-0.0057 -0.0003 -0.0532 -0.9986>}
 dyn_bag_01_r {parent=base_bag_01_r, loc=<0 0 0>, rot=<0 0 0 1>}
 base_bag_02_r {parent=dyn_bag_r, loc=<4.5952 -7.4887 3.5799>, rot=<-0.0057 -0.0003 -0.0532 -0.9986>}
 dyn_bag_02_r {parent=base_bag_02_r, loc=<0 0 0>, rot=<0 0 0 1>}
 deform_glute_base_l {parent=pelvis, loc=<-7.3466 -9.0213 -9.1245>, rot=<0.0172 0.031 -0.4477 -0.8935>}
 deform_glute_l {parent=deform_glute_base_l, loc=<0 0 0>, rot=<0 0 0 1>}
 deform_glute_base_r {parent=pelvis, loc=<-7.3464 -9.0213 9.1245>, rot=<0.031 -0.0172 0.8935 -0.4477>}
 deform_glute_r {parent=deform_glute_base_r, loc=<0 0 0>, rot=<0 0 0 1>}
 base_belt_0 {parent=pelvis, loc=<3.9166 8.7771 -0.3362>, rot=<0.3824 -0.6752 -0.5374 -0.3303>}
 dyn_belt_0 {parent=base_belt_0, loc=<0 0 0>, rot=<0 0 0 1>}
 dyn_belt_1 {parent=dyn_belt_0, loc=<8.7029 0 0>, rot=<-0.0519 -0.1022 0.0344 -0.9928>}
 dyn_belt_2 {parent=dyn_belt_1, loc=<6.7801 0 0>, rot=<-0.0921 -0.2642 0.0956 -0.9553>}
 dyn_belt_3 {parent=dyn_belt_2, loc=<6.5442 0 0>, rot=<-0.0572 -0.3533 0.1898 -0.9143>}
 dyn_belt_4 {parent=dyn_belt_3, loc=<5.8628 0 0>, rot=<0.0644 -0.2545 0.1179 -0.9577>}
 dyn_belt_5 {parent=dyn_belt_4, loc=<5.8909 0 0>, rot=<0.0067 -0.2145 0.0835 -0.9731>}
 dyn_belt_6 {parent=dyn_belt_5, loc=<6.1565 0 0>, rot=<0.1176 -0.1519 0.0444 -0.9804>}
 dyn_belt_7 {parent=dyn_belt_6, loc=<7.0363 0 0>, rot=<0 0 0 1>}
 dyn_belt_rot_6 {parent=dyn_belt_6, loc=<0 0 0>, rot=<0 0 0 1>}
 dyn_belt_rot_5 {parent=dyn_belt_5, loc=<0 0 0>, rot=<0 0 0 1>}
 dyn_belt_rot_4 {parent=dyn_belt_4, loc=<0 0 0>, rot=<0 0 0 1>}
 aim_up_bag_01 {parent=dyn_belt_rot_4, loc=<0.0001 -8.8934 0>, rot=<0 0 0 1>}
 dyn_belt_rot_3 {parent=dyn_belt_3, loc=<0 0 0>, rot=<0 0 0 1>}
 base_bag_01_l {parent=dyn_belt_3, loc=<2.7815 -0.4676 0.6071>, rot=<-0.4902 -0.5185 0.5487 0.4356>}
 dyn_bag_01_l {parent=base_bag_01_l, loc=<0 0 0>, rot=<0 0 0 1>}
 base_hinge_bag_01_l {parent=dyn_bag_01_l, loc=<-0.1773 -5.277 -0.0259>, rot=<-0.0171 -0.0023 0.1361 -0.9905>}
 dyn_hinge_bag_01_l {parent=base_hinge_bag_01_l, loc=<0 0 0>, rot=<0 0 0 1>}
 base_zipper_bag_01_l {parent=dyn_bag_01_l, loc=<0.1402 -2.1087 -4.446>, rot=<-0.7099 -0.05 0.1366 -0.6891>}
 dyn_zipper_bag_01_l {parent=base_zipper_bag_01_l, loc=<0 0 0>, rot=<0 0 0 1>}
 dyn_belt_rot_2 {parent=dyn_belt_2, loc=<0 0 0>, rot=<0 0 0 1>}
 aim_up_bag_02 {parent=dyn_belt_rot_2, loc=<0 -4.9824 0>, rot=<0 0 0 1>}
 base_bag_02_l {parent=dyn_belt_2, loc=<0.3032 1.4652 0.8215>, rot=<0.5861 0.4621 -0.412 -0.5226>}
 dyn_bag_02_l {parent=base_bag_02_l, loc=<0 0 0>, rot=<0 0 0 1>}
 base_zipper_bag_02_l {parent=dyn_bag_02_l, loc=<-0.1631 -2.1425 -4.4227>, rot=<-0.6986 -0.0766 0.1607 -0.6931>}
 dyn_zipper_bag_02_l {parent=base_zipper_bag_02_l, loc=<0 0 0>, rot=<0 0 0 1>}
 base_hinge_bag_02_l {parent=dyn_bag_02_l, loc=<-0.1991 -5.2659 0.0146>, rot=<0.6894 -0.1078 0.065 -0.7133>}
 dyn_hinge_bag_02_l {parent=base_hinge_bag_02_l, loc=<0 0 0>, rot=<0 0 0 1>}
 dyn_belt_rot_1 {parent=dyn_belt_1, loc=<0 0 0>, rot=<0 0 0 1>}
 dyn_belt_rot_0 {parent=dyn_belt_0, loc=<0 0 0>, rot=<0 0 0 1>}
 base_spray_02 {parent=pelvis, loc=<10.1769 -9.2845 11.8091>, rot=<0.1775 -0.4665 0.8616 0.0925>}
 dyn_spray_02 {parent=base_spray_02, loc=<0 0 0>, rot=<0 0 0 1>}
 base_spray_01 {parent=pelvis, loc=<7.9931 -12.7189 7.5073>, rot=<0.233 -0.5056 0.8249 0.0983>}
 dyn_spray_01 {parent=base_spray_01, loc=<0 0 0>, rot=<0 0 0 1>}
 base_grenade_02 {parent=pelvis, loc=<3.9555 -13.8653 2.1104>, rot=<0.2432 -0.5066 0.8255 0.0532>}
 dyn_grenade_02 {parent=base_grenade_02, loc=<0 0 0>, rot=<0 0 0 1>}
 base_grenade_01 {parent=pelvis, loc=<1.7634 -14.5915 -2.119>, rot=<0.2216 -0.6991 0.6698 0.1158>}
 dyn_grenade_01 {parent=base_grenade_01, loc=<0 0 0>, rot=<0 0 0 1>}
 attach {parent=root, loc=<0 0 -10.211>, rot=<0 0 -0.4694 -0.883>}
 ik_foot_root {parent=root, loc=<-2.9864 0.2357 -16.4247>, rot=<-0.0092 -0.0517 -0.0974 -0.9939>}
 ik_foot_l {parent=ik_foot_root, loc=<17.7728 -10.6802 6.6628>, rot=<-0.0082 0.6412 -0.003 -0.7673>}
 ik_foot_r {parent=ik_foot_root, loc=<-3.5035 -0.0651 12.4639>, rot=<0.6941 0.3109 -0.6192 0.1953>}
 ik_hand_root {parent=root, loc=<0 0 0>, rot=<0 0 0 1>}
 ik_hand_gun {parent=ik_hand_root, loc=<-48.1744 3.3546 104.4381>, rot=<0.6276 0.1309 0.4183 -0.6434>}
 ik_hand_l {parent=ik_hand_gun, loc=<59.3272 67.6997 34.3524>, rot=<-0.6158 -0.7027 -0.3565 0>}
 ik_hand_r {parent=ik_hand_gun, loc=<0 0 0>, rot=<0 0 0 1>}
 thigh_l_ctrl {controlType=EULER_TRANSFORM, type=Control, loc=<0 0 0>, rot=<0 0 0 1>}
 dyn_belt_0_ctrl_null {type=Null, loc=<0.3362 8.3817 101.6568>, rot=<0.663 -0.2458 -0.1052 -0.6992>}
 dyn_belt_0_ctrl {controlType=EULER_TRANSFORM, parent=dyn_belt_0_ctrl_null, type=Control, loc=<0 0 0>, rot=<0 0 0 1>}
 dyn_belt_1_ctrl_null {type=Null, loc=<7.7945 6.8256 97.4506>, rot=<0.6412 -0.2981 0.0001 -0.7071>}
 dyn_belt_1_ctrl {controlType=EULER_TRANSFORM, parent=dyn_belt_1_ctrl_null, type=Control, loc=<0 0 0>, rot=<0 0 0 1>}
 dyn_belt_2_ctrl_null {type=Null, loc=<13.3694 4.2327 94.5927>, rot=<0.5759 -0.4103 0.2646 -0.6557>}
 dyn_belt_2_ctrl {controlType=EULER_TRANSFORM, parent=dyn_belt_2_ctrl_null, type=Control, loc=<0 0 0>, rot=<0 0 0 1>}
 dyn_belt_3_ctrl_null {type=Null, loc=<16.7939 -1.1307 93.0653>, rot=<0.4734 -0.4824 0.5932 -0.4373>}
 dyn_belt_3_ctrl {controlType=EULER_TRANSFORM, parent=dyn_belt_3_ctrl_null, type=Control, loc=<0 0 0>, rot=<0 0 0 1>}
 dyn_belt_4_ctrl_null {type=Null, loc=<15.8013 -6.8504 93.8851>, rot=<0.3875 -0.5556 0.7091 -0.1956>}
 dyn_belt_4_ctrl {controlType=EULER_TRANSFORM, parent=dyn_belt_4_ctrl_null, type=Control, loc=<0 0 0>, rot=<0 0 0 1>}
 dyn_belt_5_ctrl_null {type=Null, loc=<12.1303 -11.0213 95.8419>, rot=<0.2727 -0.555 0.7858 -0.0093>}
 dyn_belt_5_ctrl {controlType=EULER_TRANSFORM, parent=dyn_belt_5_ctrl_null, type=Control, loc=<0 0 0>, rot=<0 0 0 1>}
 dyn_belt_6_ctrl_null {type=Null, loc=<6.8907 -12.9755 98.4168>, rot=<0.1737 -0.6259 0.7469 0.1421>}
 dyn_belt_6_ctrl {controlType=EULER_TRANSFORM, parent=dyn_belt_6_ctrl_null, type=Control, loc=<0 0 0>, rot=<0 0 0 1>}
 dyn_belt_7_ctrl_null {type=Null, loc=<0.5631 -13.0116 101.4943>, rot=<0.1737 -0.6259 0.7469 0.1421>}
 dyn_belt_7_ctrl {controlType=EULER_TRANSFORM, parent=dyn_belt_7_ctrl_null, type=Control, loc=<0 0 0>, rot=<0 0 0 1>}
]
variables [
 controls {
  type="TArray<FRigElementKey>"
  default="((Type=Control,Name=\"dyn_belt_0_ctrl\"),(Type=Control,Name=\"dyn_belt_1_ctrl\"),(Type=Control,Name=\"dyn_belt_2_ctrl\"),(Type=Control,Name=\"dyn_belt_3_ctrl\"),(Type=Control,Name=\"dyn_belt_4_ctrl\"),(Type=Control,Name=\"dyn_belt_5_ctrl\"),(Type=Control,Name=\"dyn_bel ...(truncated)"
 }
 head_lookdw_alpha {type=double, default=0}
 head_lookside_l_alpha {type=double, default=0}
 head_lookside_r_alpha {type=double, default=0}
 head_lookup_alpha {type=double, default=0}
 joints belt {
  type="TArray<FRigElementKey>"
  default=((Type=Bone,Name="dyn_belt_0"),(Type=Bone,Name="dyn_belt_1"),(Type=Bone,Name="dyn_belt_2"),(Type=Bone,Name="dyn_belt_3"),(Type=Bone,Name="dyn_belt_4"),(Type=Bone,Name="dyn_belt_5"),(Type=Bone,Name="dyn_belt_6"),(Type=Bone,Name="dyn_belt_7"))
 }
]
