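# Physics asset for the FuryFax medium female player body, covering hip bags, belt pieces and both thighs, apparently for secondary motion.
# The right bag hangs from the right thigh and the two left bags from tiny base capsules, all with tight angular limits.
# Those bases and two larger belt capsules link to a tiny central belt body with free swing and twist.
# The left thigh has a capsule but no constraint.
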PhysicsAsset /BRCosmetics/Characters/Player/Female/Medium/Bodies/F_MED_FuryFax/Meshes/F_MED_FuryFax_Physics
bodies {
 base_bag_01_l [{shapeType=Capsule, length=0.505, radius=0.505, shapeName=base_bag_01_l_capsule, center=<0 2.487 0>, rotation=<rot 0 -90 0>}]
 base_bag_02_l [{shapeType=Capsule, length=0.505, radius=0.505, shapeName=base_bag_02_l_capsule, center=<0 2.2422 0>, rotation=<rot 0 -90 0>}]
 dyn_bag_01_l [{
   shapeType=Capsule
   length=6.8404
   radius=4.3481
   shapeName=dyn_bag_01_l_capsule
   center=<-0.2615 -4.6349 -0.2165>
   rotation=<rot 0 79.0773 -89.3965>
  }]
 dyn_bag_02_l [{
   shapeType=Capsule
   length=6.8685
   radius=4.3466
   shapeName=dyn_bag_02_l_capsule
   center=<-0.2312 -4.3481 -0.2246>
   rotation=<rot 0 -101.3137 -87.6027>
  }]
 dyn_bag_r [{shapeType=Capsule, length=15, radius=8, shapeName=dyn_bag_r_capsule, center=<0 -9.6415 0>}]
 dyn_belt_2 [{shapeType=Capsule, length=0.505, radius=0.505, shapeName=dyn_belt_2_capsule, rotation=<rot 0 -90 0>}]
 dyn_belt_5 [{shapeType=Capsule, length=15.2126, radius=4.1083, center=<5.5406 -4.6687 -19.7788>, rotation=<rot 0 -86.9527 -72.5039>}]
 dyn_belt_rot_4 [{shapeType=Capsule, length=19.3882, radius=5.4257, center=<12.0466 -2.3741 -10.3646>, rotation=<rot 0 -99.0878 -45.7095>}]
 thigh_l [{shapeType=Capsule, length=23.285, radius=8.463, center=<-11.9158 -1.0553 1.0108>, rotation=<rot 0 92.5115 -88.4858>}]
 thigh_r [{shapeType=Capsule, length=19.7725, radius=8.8999, center=<12.3027 0.8861 -0.364>, rotation=<rot 0 -86.5236 -91.4531>}]
}
constraints [
 {bone1Name=dyn_bag_r, bone2Name=thigh_r, swing1LimitDegrees=10, swing2LimitDegrees=15, twistLimitDegrees=8}
 {bone1Name=dyn_bag_01_l, bone2Name=base_bag_01_l, swing1LimitDegrees=15, swing2LimitDegrees=10, twistLimitDegrees=0}
 {bone1Name=dyn_bag_02_l, bone2Name=base_bag_02_l, swing1LimitDegrees=15, swing2LimitDegrees=15, twistLimitDegrees=0}
 {
  bone1Name=dyn_belt_5
  bone2Name=dyn_belt_2
  swing1LimitDegrees=45
  swing1Motion=Free
  swing2LimitDegrees=45
  swing2Motion=Free
  twistLimitDegrees=45
  twistMotion=Free
 }
 {
  bone1Name=dyn_belt_rot_4
  bone2Name=dyn_belt_2
  swing1LimitDegrees=45
  swing1Motion=Free
  swing2LimitDegrees=45
  swing2Motion=Free
  twistLimitDegrees=45
  twistMotion=Free
 }
 {
  bone1Name=base_bag_01_l
  bone2Name=dyn_belt_2
  swing1LimitDegrees=45
  swing1Motion=Free
  swing2LimitDegrees=45
  swing2Motion=Free
  twistLimitDegrees=45
  twistMotion=Free
 }
 {
  bone1Name=base_bag_02_l
  bone2Name=dyn_belt_2
  swing1LimitDegrees=45
  swing1Motion=Free
  swing2LimitDegrees=45
  swing2Motion=Free
  twistLimitDegrees=45
  twistMotion=Free
 }
]
