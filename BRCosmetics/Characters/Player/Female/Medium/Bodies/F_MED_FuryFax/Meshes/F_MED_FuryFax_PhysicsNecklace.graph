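# Physics asset for the necklace on the F_MED_FuryFax female medium player body cosmetic, simulating a short and a long chain with a medal.
# Capsule bodies on upper spine, clavicles and necklace base form the torso, with clavicles locked to spine_05.
# Short and long dyn_necklace chains branch from base_necklace through constraints with small swing limits and zero twist limits.
# Swing limits generally loosen down the long chain, giving the medal end the most movement.
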PhysicsAsset /BRCosmetics/Characters/Player/Female/Medium/Bodies/F_MED_FuryFax/Meshes/F_MED_FuryFax_PhysicsNecklace
bodies {
 base_necklace [{shapeType=Capsule, length=10, radius=5, shapeName=base_necklace_capsule, center=<1.6058 7.6366 0>, rotation=<rot 0 -90 0>}]
 clavicle_l [{shapeType=Capsule, length=7.5971, radius=5.6602, shapeName=clavicle_l_capsule, center=<6.1627 -3.8639 0>, rotation=<rot 0 -170.3753 0>}]
 clavicle_r [{shapeType=Capsule, length=7.5971, radius=5.6602, shapeName=clavicle_r_capsule, center=<-7.0406 3.8639 0>, rotation=<rot 0 -167.9963 0>}]
 dyn_necklace_long_02 [{shapeType=Capsule, length=4.505, radius=2, shapeName=dyn_necklace_long_02_capsule, center=<2.7238 -1.4 0>, rotation=<rot 0 -90 0>}]
 dyn_necklace_long_03 [{shapeType=Capsule, length=7.7416, radius=1.9104, shapeName=dyn_necklace_long_03_capsule, center=<1.0425 -1.4539 0>, rotation=<rot 0 -90 0>}]
 dyn_necklace_long_04 [{shapeType=Capsule, length=0.0559, radius=1.3776, shapeName=dyn_necklace_long_04_capsule, center=<1.7074 -0.8097 0>, rotation=<rot 0 -90 0>}]
 dyn_necklace_long_medal_05 [{
   shapeType=Capsule
   length=2.4629
   radius=1.7631
   shapeName=dyn_necklace_long_medal_05_capsule
   center=<1.8839 -1.4539 0>
   rotation=<rot 0 -90 0>
  }]
 dyn_necklace_short_01 [{shapeType=Capsule, length=8, radius=3, shapeName=dyn_necklace_short_01_capsule, center=<2.8739 -2.6802 0>, rotation=<rot 0 -90 0>}]
 dyn_necklace_short_02 [{shapeType=Capsule, length=8, radius=1.5, shapeName=dyn_necklace_short_02_capsule, center=<1.9124 -1.4 0>, rotation=<rot 0 -90 0>}]
 spine_04 [{shapeType=Capsule, length=9.3715, radius=8.6887, shapeName=spine_04_capsule, center=<0.9562 3.3839 0.1868>}]
 spine_05 [{shapeType=Capsule, length=6.5349, radius=9.8948, shapeName=spine_05_capsule, center=<-1.6483 -1.2374 -0.3239>}]
}
constraints [
 {bone1Name=spine_05, bone2Name=spine_04, swing1LimitDegrees=5, swing2LimitDegrees=5, twistLimitDegrees=0}
 {bone1Name=base_necklace, bone2Name=spine_05, swing1LimitDegrees=5, swing2LimitDegrees=5, twistLimitDegrees=0}
 {
  bone1Name=clavicle_r
  bone2Name=spine_05
  swing1LimitDegrees=45
  swing1Motion=Locked
  swing2LimitDegrees=45
  swing2Motion=Locked
  twistLimitDegrees=45
  twistMotion=Locked
 }
 {
  bone1Name=clavicle_l
  bone2Name=spine_05
  swing1LimitDegrees=45
  swing1Motion=Locked
  swing2LimitDegrees=45
  swing2Motion=Locked
  twistLimitDegrees=45
  twistMotion=Locked
 }
 {bone1Name=dyn_necklace_short_01, bone2Name=base_necklace, swing1LimitDegrees=5, swing2LimitDegrees=1, twistLimitDegrees=0}
 {bone1Name=dyn_necklace_short_02, bone2Name=dyn_necklace_short_01, swing1LimitDegrees=5, swing2LimitDegrees=10, twistLimitDegrees=0}
 {bone1Name=dyn_necklace_long_02, bone2Name=base_necklace, swing1LimitDegrees=5, swing2LimitDegrees=5, twistLimitDegrees=0}
 {bone1Name=dyn_necklace_long_03, bone2Name=dyn_necklace_long_02, swing1LimitDegrees=10, swing2LimitDegrees=5, twistLimitDegrees=0}
 {bone1Name=dyn_necklace_long_04, bone2Name=dyn_necklace_long_03, swing1LimitDegrees=25, swing2LimitDegrees=20, twistLimitDegrees=0}
 {bone1Name=dyn_necklace_long_medal_05, bone2Name=dyn_necklace_long_04, swing1LimitDegrees=35, swing2LimitDegrees=20, twistLimitDegrees=0}
]
